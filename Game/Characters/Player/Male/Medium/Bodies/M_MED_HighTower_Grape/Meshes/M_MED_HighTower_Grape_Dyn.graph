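# Physics asset named for the HighTower Grape medium male player body, with a Dyn suffix.
# It holds a single body on the root bone, made of one very small capsule pitched 90 degrees.
# It defines no constraints and no bodies for other bones, so it provides no jointed ragdoll or secondary-motion setup.
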PhysicsAsset /Game/Characters/Player/Male/Medium/Bodies/M_MED_HighTower_Grape/Meshes/M_MED_HighTower_Grape_Dyn
bodies {root [{shapeType=Capsule, length=0.505, radius=0.505, rotation=<rot 0 -90 0>}]}
constraints []
